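# Translucent glass material for the Rift Shard pickup, with glowing cracks and refraction that drops to neutral on low quality.
# A shard mask texture sets crack detail and which areas wobble around Z and drift on panning noise.
# Rotation-corrected normals drive a sphere-map reflection, added to geode crack colors and Fresnel for glow and opacity.
# Vertex motion also layers in pickup LOD/distance-cull and black-hole functions.
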Material /Game/Athena/Items/Consumables/RiftShard/Materials/M_Shard_Portal_Pickup
graph {
 properties {
  bUsedWithMeshDeformer=true
  bUsedWithSkeletalMesh=true
  blendMode=BLEND_Translucent
  pixelDepthOffsetMode=PDOM_Legacy
  translucencyLightingMode=TLM_VolumetricPerVertexNonDirectional
  usedShadingModels=Default Lit
 }
 nodes [
  n1 = Constant3Vector() {constant=<rgba 0 0 1 0>}
  n2 = TextureSample() {texture=/Game/Athena/Items/Consumables/RiftShard/Textures/T_SwiftRift_Shards_M}
  n3 = Multiply(A=Time()) {constB=0.8}
  n4 = LinearInterpolate(Alpha=@n2.B) {constA=-1, constB=2}
  n5 = ScalarParameter() {defaultValue=0.07, parameterName=TurnAmount}
  n6 = LinearInterpolate(Alpha=@n2.B) {constA=0.3}
  n7 = Multiply(A=Sine(None=Add(A=@n3, B=@n4)) {period=5.8}, B=Add(A=@n5, B=Multiply(A=@n5, B=@n6)) {constB=0.05}) {constB=0.1}
  n8 = Multiply(A=Ceil(None=@n2.B), B=@n7)
  n9 = ObjectPositionWS()
  n10 = Transform(None=ReflectionVectorWS(CustomWorldNormal=MaterialFunctionCall(Rotation Axis=@n1, Rotation Angle=@n8, Normals Pivot=@n9) {materialFunction=&FixRotateAboutAxisNormals}.New World Space Vertex Normal)) {transformSourceType=TRANSFORMSOURCE_World, transformType=TRANSFORM_View}
  n11 = Multiply(
   A=ConstantBiasScale(None=ComponentMask(None=@n10) {g=true, r=true})
   B=ComponentMask(None=VectorParameter() {parameterName=CubeMapOffset, defaultValue=<rgba 1 -1 0 1>}.RGB) {g=true, r=true}
  )
  n12 = Multiply(
   A=Multiply(
    A=TextureSampleParameter2D(UVs=@n11) {parameterName=SphereReflectionMap, texture=/Game/Items/Art_noLOD/Textures/T_Items_MineralPowder_Glass_Reflection}.RGB
    B=ScalarParameter() {defaultValue=1, group=reflection, parameterName=Reflection Multiplier}
   )
   B=ComponentMask(None=MaterialFunctionCall() {materialFunction=/Engine/Functions/GetPostProcessFunctions/GetAmbientCubemapTint}.AmbientCubemapTint) {b=true, g=true, r=true}
  )
  n13 = LinearInterpolate(
   B=ScalarParameter() {defaultValue=1, parameterName=MipLevel_CrackDetail}
   Alpha=MaterialFunctionCall() {materialFunction=&Near_Camera_Fade}.Result
  )
  n14 = TextureSample!(UVs, Tex, Level=@n13, Apply View MipBias) {mipValueMode=TMVM_MipLevel, texture=&T_SwiftRift_Shards_M}
  n15 = LinearInterpolate(Alpha=@n14.G) {constA=0.22}
  n16 = Fresnel(ExponentIn=ScalarParameter() {defaultValue=2.5, parameterName=Fresnel Exponent}) {baseReflectFraction=0, exponent=4}
  n17 = MaterialFunctionCall!(Use Alternate UVs, Alternate UVs) {materialFunction=/Game/Athena/Effects/Materials/Geode/Functions/MF_Geode_Crack_Colors}
  n18 = Multiply(A=Multiply(A=@n15, B=@n16), B=Add(A=@n17.Cloud Color) {constA=0.2, constB=0.4})
  n19 = Multiply(A=Multiply(A=@n17.Crack Color) {constB=3}, B=@n14.G)
  Add(A=@n12, B=Multiply(A=ScalarParameter() {defaultValue=1, parameterName=Emissive}, B=Add(A=@n18, B=@n19)))
  n20 = Ceil(None=TextureSample(Level=@n13) {mipValueMode=TMVM_MipLevel, texture=&T_SwiftRift_Shards_M}.B)
  n21 = RotateAboutAxis(NormalizedRotationAxis=@n1, RotationAngle=@n7, PivotPoint=@n9, Position=WorldPosition().XYZ)
  n22 = TextureCoordinate() {uTiling=0.01, vTiling=0.01}
  n23 = TextureSample!(UVs=Panner(Coordinate=Add(A=@n22, B=@n2.B)) {speedX=0.01, speedY=0.015}, Tex, Apply View MipBias) {samplerType=SAMPLERTYPE_LinearColor, texture=&HighContrastNoise}
  n24 = Multiply(A=Multiply(A=Subtract(A=@n23.RGB) {constB=0.5}) {constB=2}, B=ScalarParameter() {defaultValue=2, parameterName=DriftDistance})
  n25 = VectorParameter() {parameterName=WorldPostionOfBlackHole, defaultValue=<rgba -30000 25536 1952 0>}
  n26 = AppendVector(A=@n25.RGB, B=@n25.RGB)
  n27 = MaterialFunctionCall!(
   World Position
   DistanceCullFadeAnimationOffset=MaterialFunctionCall() {materialFunction=/Game/Packages/Fortress_SharedMaterials/Base_Material_Functions/PickupEffectDistanceCullOverride}.Result
   Use actor center instead of pivot point
  ) {materialFunction=/Game/Packages/Fortress_SharedMaterials/Base_Material_Functions/LODEffect}
  n28 = Add(A=@n27.Result, B=WorldPosition().XYZ)
  Add(
   A=Multiply(A=@n20, B=Add(A=@n21, B=@n24))
   B=Add(
    A=MaterialFunctionCall!(WorldPostionOfBlackHole=@n26, World Position=@n28) {materialFunction=/Game/Packages/Fortnite_UniqueMaterials_1/BlackHoleFunction}.Result
    B=@n27.Result
   )
  )
  n29 = Fresnel() {baseReflectFraction=0, exponent=4}
  Multiply(B=Add(A=Multiply(A=@n29) {constB=0.5}, B=VectorParameter() {parameterName=GlassColor, defaultValue=<rgba 0.6649 0.2702 1 0>}.RGB)) {constA=1}
  n30 = Ceil(None=@n14.B)
  n31 = Multiply(A=@n15, B=@n30)
  Multiply(A=ScalarParameter() {defaultValue=1, parameterName=Opacity}, B=Add(A=Multiply(A=@n12) {constB=0.4}, B=Add(A=@n16, B=@n31)))
  OneMinus(None=@n30)
  n32 = OneMinus(None=@n16)
  QualitySwitch(
   Default=LinearInterpolate(B=ScalarParameter() {defaultValue=1.2, parameterName=Refraction}, Alpha=Multiply(A=@n32, B=@n20)) {constA=1}
   Low=Constant() {r=1}
  )
 ]
}
