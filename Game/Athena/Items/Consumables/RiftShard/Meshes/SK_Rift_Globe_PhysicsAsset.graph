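# Physics asset for the Rift Globe skeletal mesh used by the RiftShard consumable item.
# It holds a single capsule body on the globe_jnt bone, slightly offset below the joint and rotated to fit the globe.
# There are no constraints, so the globe is one collision shape with no jointed ragdoll parts.
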PhysicsAsset /Game/Athena/Items/Consumables/RiftShard/Meshes/SK_Rift_Globe_PhysicsAsset
bodies {globe_jnt [{shapeType=Capsule, length=20.5605, radius=18.8064, center=<-0.0969 -0.0448 -3.4409>, rotation=<rot 0.7213 -90 1.7824>}]}
constraints []
